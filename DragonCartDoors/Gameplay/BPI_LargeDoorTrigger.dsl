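; BPI_LargeDoorTrigger is a Blueprint interface for triggering large dragon cart doors and their vault alarms.
; It declares calls to activate a dragon cart door with an instigator object and to reset it.
; It also declares calls to register an alarm actor with a vault door and to activate or deactivate vault alarms.
; The functions are signatures only, so implementing classes supply the behavior.
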(fn ActivateDragonCartDoor
  (params (Instigator :type Object%20Reference)))

(fn ActivateVaultAlarms)

(fn DeactivateVaultAlarms)

(fn RegisterVaultDoorAlarm
  (params (AlarmToRegister :type Actor%20Object%20Reference)))

(fn ResetDragonCartDoor)
