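; Placed trigger that activates or resets linked DragonCart large doors through player interaction or, optionally, damage, while taking no damage itself.
; On BeginPlay it registers with each target door and binds their start/finish events to drive its trigger state, with delays after completion.
; Activation requires a tagged Fort pawn and server authority; only one state permits interaction.
; A text render shows the state, and construction can draw debug lines to doors.
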(event EventBeginPlay
  (|CallParentFunction)
  (for _array_element (Variables|Default|GetTargetDoors)
    (Utilities|IsValid _array_element
      (:Is%20Valid
        (Variables|Default|SetValidDoorExists true))
      (:Is%20Not%20Valid
        (Development|PrintString "No target door set. This needs to be set on the instance in the world" true true "(R=1.000000,G=0.000000,B=0.043362,A=1.000000)"))))
  (Utilities|FlowControl|Sequence
    (:then_1
      (EventDispatchers|AssignOnDamaged (delegate Custom|OnDamaged_Event))
      (CallFunction|UpdateTriggerState "NewEnumerator1"))
    (:then_2
      (for _array_element_2 (Variables|Default|GetTargetDoors)
        (Utilities|IsValid _array_element_2
          (:Is%20Valid
            (Default|AssignOnDoorFinishActivation _array_element_2 (EventDispatchers|CreateEvent :!event Custom|OnDoorFinishedActivation))
            (Default|AssignOnDoorFinishReset _array_element_2 (EventDispatchers|CreateEvent :!event Custom|OnDoorFinishedResetting))
            (Default|BindEventtoOnDoorStartActivation _array_element_2 (EventDispatchers|CreateEvent :!event Custom|OnDoorStartActivating))
            (Default|BindEventtoOnDoorStartReset _array_element_2 (EventDispatchers|CreateEvent :!event Custom|OnDoorStartResetting))
            (Class|BPLargeDoorParent|RegisterDoorTrigger _array_element_2 self) ; Door needs to know all triggers associated with it. \nIf both need power, and one receives power, both treat that as receiving power
          ))))))

(event Custom|OnDamaged_Event
  (params (DamagedActor :type Actor%20Object%20Reference) (Damage :type Float%20%28single-precision%29) (InstigatedBy :type Controller%20Object%20Reference) (DamageCauser :type Actor%20Object%20Reference) (HitLocation :type Vector) (FHitComponent :type Primitive%20Component%20Object%20Reference) (BoneName :type Name) (Momentum :type Vector))
  (if (Variables|Default|GetActivateOnDamaged)
    (CallFunction|ActivateTargetedDoor InstigatedBy)))

(event OnLinkedActorDestroyed(FortLinkToActor)
  (params (DamageCauser :type Actor%20Object%20Reference))
  (if (not (and (Variables|Default|GetIgnoreLinkedActorDestruction) (ScalableFloat|GetValueasBool (Variables|Default|GetHF_AllowIgnoreLinkedActorDestruction))))
    (Rendering|SetActorHiddenInGame true)
    (Actor|DestroyActor)))

(event Custom|OnDoorFinishedActivation
  (Development|PrintString "Trigger sees door finished activation")
  (Utilities|FlowControl|Delay (ScalableFloat|GetValueatLevel (Variables|Default|GetStateTransitionDelay)))
  (if (Variables|DragonCartDoor|GetCanResetAfterActivation)
    (CallFunction|UpdateTriggerState "NewEnumerator4")
    (else
      (CallFunction|UpdateTriggerState "NewEnumerator5"))))

(event Custom|OnDoorFinishedResetting
  (Development|PrintString "Trigger sees door finished resetting")
  (Utilities|FlowControl|Delay (ScalableFloat|GetValueatLevel (Variables|Default|GetStateTransitionDelay)))
  (CallFunction|UpdateTriggerState "NewEnumerator1") ; If this trigger needs power, resetting it is removing its power source
)

(event Custom|OnDoorStartActivating
  (CallFunction|UpdateTriggerState "NewEnumerator2"))

(event Interaction|EventBlueprintOnInteract
  (params (InteractingPawn :type Fort%20Pawn%20Object%20Reference) (InteractionBeingAttempted :type EInteractionBeingAttempted%20Enum :default "FirstInteraction"))
  (switch Utilities|FlowControl|Switch|SwitchonDragonCart_DoorTriggerStates (Variables|Default|GetTriggerState)
    (:NewEnumerator1
      (CallFunction|ActivateTargetedDoor InteractingPawn)
      (GameplayCue|ExecuteGameplayCueOnActor(Burst) self (Variables|Default|GetGCButtonInteractSound) (Ability|GameplayCue|MakeGameplayCueParameters 0.0)))
    (:NewEnumerator4
      (CallFunction|ResetTargetedDoor))))

(event Custom|OnDoorStartResetting
  (CallFunction|UpdateTriggerState "NewEnumerator3"))

(event Custom|PlayInteractAnimation
  (params (FortPlayerPawn :type Fort%20Player%20Pawn%20Object%20Reference)))

(fn ActivateTargetedDoor
  (params (InInstigator :type Object%20Reference) (EnableAlternateOpen :type Boolean))
  (LocalVariables|SetEnableAlternateOpenLocal EnableAlternateOpen)
  (switch Utilities|FlowControl|Switch|SwitchonDragonCart_DoorTriggerStates (Variables|Default|GetTriggerState)
    (:NewEnumerator1
      (bind _fortpawn (Class|FortQuestBlueprintLibrary|GetFortPawnFromObject InInstigator)
        (:Valid
          (bind _returnvalue (GameplayTags|HasMatchingGameplayTag(Message) _fortpawn "(TagName=\"Homebase.Class\")"))
          (if _returnvalue
            (Networking|SwitchHasAuthority
              (:Authority
                (for _array_element (Variables|Default|GetTargetDoors)
                  (Utilities|IsValid _array_element
                    (:Is%20Valid
                      (Class|BPILargeDoorTrigger|ActivateDragonCartDoor _array_element 0)
                      (Variables|Default|SetTriggerActivated true)
                      (CallFunction|UpdateTriggerState "NewEnumerator2"))))))
            (else
              (Development|PrintString "Not Pawn Damage On Door Trigger???" true true "(R=1.000000,G=0.533333,B=0.000000,A=1.000000)"))))
        (:NotValid
          (Development|PrintString "Not a Fort Pawn???" true true "(R=1.000000,G=0.533333,B=0.000000,A=1.000000)"))))
    (:NewEnumerator4
      (bind _fortpawn_2 (Class|FortQuestBlueprintLibrary|GetFortPawnFromObject InInstigator)
        (:Valid
          (bind _returnvalue_2 (GameplayTags|HasMatchingGameplayTag(Message) _fortpawn_2 "(TagName=\"Homebase.Class\")"))
          (if _returnvalue_2
            (Networking|SwitchHasAuthority
              (:Authority
                (CallFunction|ResetTargetedDoor)))
            (else
              (Development|PrintString "Not Pawn Damage On Door Trigger???" true true "(R=1.000000,G=0.533333,B=0.000000,A=1.000000)"))))
        (:NotValid
          (Development|PrintString "Not a Fort Pawn???" true true "(R=1.000000,G=0.533333,B=0.000000,A=1.000000)"))))))

(fn BlueprintCanInteract
  (params (InteractingPawn :type Fort%20Pawn%20Object%20Reference) (InteractionBeingAttempted :type EInteractionBeingAttempted%20Enum :default "FirstInteraction") (InteractionType :type TInteractionType%20Enum :default "IT_NoInteraction"))
  (switch Utilities|FlowControl|Switch|SwitchonDragonCart_DoorTriggerStates (Variables|Default|GetTriggerState)
    (:NewEnumerator0
      _)
    (:NewEnumerator1
      (return true))
    (:NewEnumerator2
      _)
    (:NewEnumerator4
      _)
    (:NewEnumerator3
      _))
  (return false))

(fn BlueprintGetInteractionString
  (params (InteractingPawn :type Fort%20Pawn%20Object%20Reference) (InteractionBeingAttempted :type EInteractionBeingAttempted%20Enum :default "FirstInteraction"))
  (switch Utilities|FlowControl|Switch|SwitchonDragonCart_DoorTriggerStates (Variables|Default|GetTriggerState)
    (:NewEnumerator1
      ;[1 1 1] Open
      (return "NSLOCTEXT(\"[E2100534C323408FB713B0A5EC747545]\", \"36B3A6284A8DAC2D25773E9407EBFFC9\", \"Activate Door\")"))
    (:NewEnumerator4
      ;[1 1 1] Close
      (return "NSLOCTEXT(\"[E2100534C323408FB713B0A5EC747545]\", \"E1DA4AC047874F2236B6139E66BD695F\", \"Activate Door\")"))))

(fn BlueprintModifyIncomingDamage
  (params (Damage :type Float%20%28single-precision%29) (DamageTags :type Gameplay%20Tag%20Container%20Structure%20%28by%20ref%29) (EffectContext :type Gameplay%20Effect%20Context%20Handle%20Structure%20%28by%20ref%29) (EventInstigator :type Controller%20Object%20Reference) (DamageCauser :type Actor%20Object%20Reference))
  (return 0.0))

(fn OnRep_TriggerActivated
  (return))

(fn OnRep_TriggerState
  (CallFunction|UpdateTriggerStateText (Variables|Default|GetTriggerState)))

(fn ResetTargetedDoor
  (for _array_element (Variables|Default|GetTargetDoors)
    (Utilities|IsValid _array_element
      (:Is%20Valid
        (Class|BPILargeDoorTrigger|ResetDragonCartDoor _array_element))))
  (Variables|Default|SetTriggerActivated true)
  (CallFunction|UpdateTriggerState "NewEnumerator3")
  (return))

(fn UpdateTriggerState
  (params (NewState :type DragonCart_DoorTriggerStates%20Enum :default "NewEnumerator0"))
  (bind _output_get (Variables|Default|SetTriggerState NewState))
  (CallFunction|UpdateTriggerStateText _output_get))

(fn UpdateTriggerStateText
  (params (Selection :type DragonCart_DoorTriggerStates%20Enum :default "NewEnumerator0"))
  (Rendering|Components|TextRender|SetText (Variables|Default|GetBoxStateText) (Utilities|Text|ToText(String) (Utilities|String|EnumtoString Selection))))

(fn ConstructionScript
  (Utilities|FlowControl|Sequence
    (:then_0
      (Rendering|SetVisibility (Variables|Default|GetAlignmentCylinder) (Variables|Default|GetShowHeightAssist) false))
    (:then_1
      (if (Variables|Default|GetShowConnectedDoors)
        (for _array_element (Variables|Default|GetTargetDoors)
          (Utilities|IsValid _array_element
            (:Is%20Valid
              (Rendering|Debug|DrawDebugLine (Transformation|GetActorLocation) (Math|Vector|MakeVector (.x (Transformation|GetActorLocation _array_element)) (.y (Transformation|GetActorLocation _array_element)) (+ (.z (Transformation|GetActorLocation _array_element)) (* (.z (.BoxExtent (Collision|GetActorBounds _array_element false false))) 0.5))) "(R=0.833333,G=0.000000,B=1.000000,A=1.000000)" 5.0 2.0))))))))
