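; Base Blueprint for Dragon Cart large doors that activate, reset, and can damage players and vehicles they smash.
; Authority-only events advance a door state enum and stamp server time, while timeline callbacks finalize state and physical open status.
; RepNotifies drive movement, gameplay cues, sounds and victim/instigator hit audio.
; Movement and smash-target queries are stubs for children, with hotfix values controlling player damage and vehicle effects.
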(event EventBeginPlay
  (Utilities|FlowControl|Sequence
    (:then_0
      (if (Variables|Default|GetOverrideDoorHealth)
        (Gameplay|ForceBuildingHealth self (Math|Conversions|ToFloat(Integer) (Variables|Default|GetDoorHealth)))
        (else
          (CallFunction|SetDoorHealthfromDataTable))))
    (:then_1
      (CallFunction|MoveDoorbyPercentage (Variables|Default|GetDoorOperationPercentage))
      (Variables|Default|SetEditor_PanelPercentage 0.0)
      (Variables|Default|SetbIsOpen (not (Variables|Default|GetbStartsClosed))))))

(event EventActivateDragonCartDoor
  (params (Instigator :type Object%20Reference))
  (Variables|Default|SetDoorInstigator Instigator)
  (Networking|SwitchHasAuthority
    (:Authority
      (switch Utilities|FlowControl|Switch|SwitchonLargeDoorState (Variables|Default|GetDoorState)
        (:NewEnumerator0
          (Variables|Default|SetDoorState "NewEnumerator1")
          (Development|PrintString "Door Parent is Trying to Activate Door" true true "(R=1.000000,G=0.533333,B=0.000000,A=1.000000)")
          (Variables|Default|SetStartMoveTime (Fortnite|Sync|Time|GetServerWorldTimeSeconds)))))))

(event EventResetDragonCartDoor
  (Networking|SwitchHasAuthority
    (:Authority
      (switch Utilities|FlowControl|Switch|SwitchonLargeDoorState (Variables|Default|GetDoorState)
        (:NewEnumerator2
          (Variables|Default|SetDoorState "NewEnumerator3")
          (Development|PrintString "Door Parent is Trying to Reset Door" true true "(R=1.000000,G=0.533333,B=0.000000,A=1.000000)")
          (Variables|Default|SetStartMoveTime (Fortnite|Sync|Time|GetServerWorldTimeSeconds)))))))

(event EventDestroyed
  (|CallParentFunction)
  (if (Variables|Default|GetDestroySwitchOnDeath)
    _))

(event Custom|SetDoorHealthFromDataTable)

(event EventOnDeathServer
  (params (Damage :type Float%20%28single-precision%29) (DamageTags :type Gameplay%20Tag%20Container%20Structure%20%28by%20ref%29) (Momentum :type Vector :default "0, 0, 0") (HitInfo :type Hit%20Result%20Structure%20%28by%20ref%29) (InstigatedBy :type Controller%20Object%20Reference) (DamageCauser :type Actor%20Object%20Reference) (EffectContext :type Gameplay%20Effect%20Context%20Handle%20Structure))
  (|CallParentFunction Damage DamageTags Momentum HitInfo InstigatedBy DamageCauser EffectContext)
  (Rendering|SetActorHiddenInGame true))

(fn BP_OnActive
  (params (EnableAlternateActive :type Boolean) (EnableOpenCues :type Boolean))
  (bind _soundlibrarycomponent (Variables|Default|GetSoundLibraryComponent))
  (if EnableOpenCues
    (Default|CallOnDoorStartActivation)
    (GameplayCue|AddGameplayCue self (Variables|Default|GetGC_DoorActive))
    (Networking|CanExecuteCosmeticEvents
      (:True
        (SoundLibrary|PlayEvent _soundlibrarycomponent "(TagName=\"SoundLibrary.Door.Active\")")
        (SoundLibrary|StopEvent _soundlibrarycomponent "(TagName=\"SoundLibrary.Door.Reset\")")
        (SoundLibrary|PlayEvent _soundlibrarycomponent "(TagName=\"SoundLibrary.Door.Open\")")))))

(fn BP_OnClosed
  (params (EnableAlternateClose :type Boolean) (EnableOpenCues :type Boolean))
  (bind _soundlibrarycomponent (Variables|Default|GetSoundLibraryComponent))
  (if EnableOpenCues
    (Ability|ExecuteGameplayCueReplicated (Variables|Default|GetGC_DoorClosed))
    (Networking|CanExecuteCosmeticEvents
      (:True
        (SoundLibrary|PlayEvent _soundlibrarycomponent "(TagName=\"SoundLibrary.Door.Close\")")
        (SoundLibrary|StopEvent _soundlibrarycomponent "(TagName=\"SoundLibrary.Door.Active\")")
        (SoundLibrary|StopEvent _soundlibrarycomponent "(TagName=\"SoundLibrary.Door.Reset\")")))))

(fn BP_OnOpen
  (params (EnableAlternateOpen :type Boolean) (EnableOpenCues :type Boolean))
  (bind _soundlibrarycomponent (Variables|Default|GetSoundLibraryComponent))
  (if EnableOpenCues
    (Ability|ExecuteGameplayCueReplicated (Variables|Default|GetGC_DoorOpen))
    (Networking|CanExecuteCosmeticEvents
      (:True
        (SoundLibrary|PlayEvent _soundlibrarycomponent "(TagName=\"SoundLibrary.Door.Open\")")
        (SoundLibrary|StopEvent _soundlibrarycomponent "(TagName=\"SoundLibrary.Door.Active\")")
        (SoundLibrary|StopEvent _soundlibrarycomponent "(TagName=\"SoundLibrary.Door.Reset\")")))))

(fn BP_OnReset
  (params (EnableAlternateReset :type Boolean) (EnableOpenCues :type Boolean))
  (bind _soundlibrarycomponent (Variables|Default|GetSoundLibraryComponent))
  (if EnableOpenCues
    (Default|CallOnDoorStartReset)
    (Networking|CanExecuteCosmeticEvents
      (:True
        (SoundLibrary|PlayEvent _soundlibrarycomponent "(TagName=\"SoundLibrary.Door.Reset\")")
        (SoundLibrary|StopEvent _soundlibrarycomponent "(TagName=\"SoundLibrary.Door.Active\")")
        (SoundLibrary|PlayEvent _soundlibrarycomponent "(TagName=\"SoundLibrary.Door.Close\")")))))

(fn GetDoorTriggers
  (return (Variables|Default|GetRegisteredDoorTriggers)))

(fn GetPlayersToSmash
  ; To be implemented in children
  (return 0))

(fn GetVehiclesToSmash
  ; Implemented in children
  (return 0))

(fn IsLargeDoorOpen
  ; Since "open" and "closed" states unfortunately mean different things for doors that start closed vs open, this bool represents if the door is actually physically open
  (return (Variables|Default|GetbIsOpen)))

(fn MoveDoorByPercentage
  (params (Percentage :type Float%20%28double-precision%29)))

(fn OnDoorTimelineFinished
  (params (Direction :type ETimelineDirection%20Enum :default "Forward"))
  (switch Utilities|FlowControl|Switch|SwitchonETimelineDirection Direction
    (:Forward
      (Networking|SwitchHasAuthority
        (:Authority
          (Variables|Default|SetDoorState "NewEnumerator2")
          (Default|CallOnDoorFinishActivation)
          (Variables|Default|GetbStartsClosed
            (:then
              (Variables|Default|SetbIsOpen true))
            (:else
              (Variables|Default|SetbIsOpen false))))))
    (:Backward
      (Networking|SwitchHasAuthority
        (:Authority
          (Variables|Default|SetDoorState "NewEnumerator0")
          (Default|CallOnDoorFinishReset)
          (Variables|Default|GetbStartsClosed
            (:then
              (Variables|Default|SetbIsOpen false))
            (:else
              (Variables|Default|SetbIsOpen true))))))))

(fn OnDoorTimelineStarted
  (params (Direction :type ETimelineDirection%20Enum :default "Forward"))
  (switch Utilities|FlowControl|Switch|SwitchonETimelineDirection Direction
    (:Forward
      (Default|CallOnDoorStartActivation))
    (:Backward
      (Default|CallOnDoorStartReset))))

(fn OnRep_DoorInstigator
  (bind _asplayer_controller (Utilities|Casting|CastToPlayerController (Variables|Default|GetDoorInstigator)))
  (Variables|Default|SetbIsLocalInstigator (Pawn|IsLocallyViewed (Pawn|GetControlledPawn _asplayer_controller))))

(fn OnRep_DoorOperationPercentage
  (CallFunction|MoveDoorbyPercentage (Variables|Default|GetDoorOperationPercentage)))

(fn OnRep_DoorState
  (switch Utilities|FlowControl|Switch|SwitchonLargeDoorState (Variables|Default|GetDoorState)
    (:NewEnumerator0
      (CallFunction|BPonOpen false (Variables|Default|GetEnableGameplayCues)))
    (:NewEnumerator1
      (CallFunction|BPonActive false (Variables|Default|GetEnableGameplayCues)))
    (:NewEnumerator2
      (CallFunction|BPonClosed false (Variables|Default|GetEnableGameplayCues)))
    (:NewEnumerator3
      (Default|BPonReset false (Variables|Default|GetEnableGameplayCues)))))

(fn OnRep_OverlappingPawnActors
  (bind _overlappingpawnactors (Variables|Default|GetOverlappingPawnActors))
  (Networking|CanExecuteCosmeticEvents
    (:True
      (if (Utilities|Array|IsNotEmpty _overlappingpawnactors)
        (for _array_element _overlappingpawnactors
          (bind _asfort_pawn (Utilities|Casting|CastToFortPawn _array_element))
          (Utilities|IsValid (Pawn|GetLocalViewingPlayerController _asfort_pawn)
            (:Is%20Valid
              (if (Pawn|IsLocalController (Pawn|GetLocalViewingPlayerController _asfort_pawn))
                (Audio|PlaySoundatLocation "/Game/Athena/Sounds/HitNotifications/Gunplay_Update/MSS_HitNotifies_GunplayUpdate_Receiver_Downed.MSS_HitNotifies_GunplayUpdate_Receiver_Downed" (Transformation|GetActorLocation))
                (else
                  _)))
            (:Is%20Not%20Valid
              _))
          (Variables|Default|GetbIsLocalInstigator
            (:then
              (Audio|PlaySoundatLocation "/Game/Athena/Sounds/HitNotifications/Gunplay_Update/MSS_HitNotifies_GunplayUpdate_Instigator_Downed.MSS_HitNotifies_GunplayUpdate_Instigator_Downed" (Transformation|GetActorLocation _asfort_pawn)))
            (:else)))))))

(fn RegisterDoorTrigger
  (params (DoorTrigger :type BP%20Dragon%20Cart%20Large%20Door%20Trigger%20Object%20Reference))
  (Utilities|Array|Add (Variables|Default|GetRegisteredDoorTriggers) DoorTrigger))

(fn SmashPlayers
  (params (OverlappingActors :type Array%20of%20Actor%20Object%20References))
  (bind _output_get (Variables|Default|SetOverlappingPawnActors OverlappingActors))
  (for _array_element _output_get
    (bind _asfort_player_pawn (Utilities|Casting|CastToFortPlayerPawn _array_element))
    (Game|Damage|ApplyDamage _asfort_player_pawn (ScalableFloat|GetValueatLevel (Variables|Hotfix|GetHF_SmashPlayers_Damage)) 0 self "/Engine/EngineDamageTypes/DmgTypeBP_Environmental.DmgTypeBP_Environmental_C")))

(fn SmashVehicles
  (params (OverlappingActors :type Array%20of%20Actor%20Object%20References))
  (if (ScalableFloat|GetValueasBool (Variables|Hotfix|GetHF_SmashVehicles))
    (for _array_element OverlappingActors
      (bind _asfort_athena_vehicle (Utilities|Casting|CastToFortAthenaVehicle _array_element))
      (Fortnite|ApplyGameplayEffectsToActor _asfort_athena_vehicle (Utilities|Array|MakeArray (Variables|Default|GetVehicleDamageGE)) (Transformation|GetActorLocation) false 0 (Ability|GetAbilitySystemComponent 0)))))

(fn ConstructionScript)
